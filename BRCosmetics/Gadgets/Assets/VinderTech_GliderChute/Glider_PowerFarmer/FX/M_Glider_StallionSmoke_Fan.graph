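# Unlit translucent Niagara material that draws a procedural spinning fan for the Power Farmer glider's stallion smoke effects.
# Twisted polar UVs scroll over time and repeat into reflected, S-curve-sharpened gradients that form the blades.
# Soft center and edge masks trim the disc, and a dynamic parameter fills the gaps between blades.
# Particle color feeds the emissive adjustment function, and particle alpha scales the mask.
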Material /BRCosmetics/Gadgets/Assets/VinderTech_GliderChute/Glider_PowerFarmer/FX/M_Glider_StallionSmoke_Fan
graph {
 properties {
  bUsedWithNiagaraMeshParticles=true
  bUsedWithNiagaraSprites=true
  blendMode=BLEND_Translucent
  pixelDepthOffsetMode=PDOM_Legacy
  shadingModel=MSM_Unlit
  usedShadingModels=Unlit
 }
 nodes [
  n1 = ParticleColor()
  MaterialFunctionCall(Emissive=@n1.RGB) {materialFunction=&MF_FX_EmissiveAdjustments}
  n2 = MaterialFunctionCall!(UV, Twist Center=ScalarParameter() {defaultValue=0.1, parameterName=Fan Twist}, Twist Edges) {materialFunction=/Game/Accessories/FORT_Backpacks/Backpack_F_MED_Celestial_Glow/FX/Materials/MF_PolarUV}
  n3 = ComponentMask(None=@n2.Polar UVs) {g=true}
  n4 = ScalarParameter() {defaultValue=0.1, parameterName=Center Offset}
  n5 = Saturate(None=Divide(A=Subtract(A=@n3, B=@n4), B=ScalarParameter() {defaultValue=0.2, parameterName=Center Softness}))
  n6 = Saturate(None=Divide(A=OneMinus(None=@n3), B=ScalarParameter() {defaultValue=0.1, parameterName=Edge Softness}))
  n7 = Add(A=ComponentMask(None=@n2.Polar UVs) {r=true}, B=Multiply(A=Time(), B=ScalarParameter() {defaultValue=3, parameterName=Rotation Speed}))
  n8 = ScalarParameter() {defaultValue=3, parameterName=Blades}
  n9 = MaterialFunctionCall!(
   Gradient=Frac(None=Multiply(A=@n7, B=@n8) {constB=5})
   Center Point=ScalarParameter() {defaultValue=0.5, parameterName=Gradient Center Point}
  ) {materialFunction=/Game/Effects/Fort_Effects/MaterialFunctions/MF_GradientReflected}
  n10 = ScalarParameter() {defaultValue=3, parameterName=Gradient Exponent}
  n11 = DynamicParameter() {defaultValue=<rgba 0 1 1 1>, paramNames [Background Opacity, Param2, Param3, Param4]}
  n12 = Multiply(
   A=OneMinus(None=MaterialFunctionCall(In=@n9.Result, Power=@n10) {materialFunction=&SCurve}.Result)
   B=OneMinus(None=Saturate(None=@n11.Background Opacity))
  )
  Multiply(A=@n1.A, B=Multiply(A=Multiply(A=@n5, B=@n6), B=OneMinus(None=@n12)))
 ]
}
